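# Physics asset for the F_MED_SpatialTravel_Quirk female medium player body cosmetic, set up for secondary motion on its jacket.
# Seven jacket chains (three back, two front, two side) each pair a tiny base capsule with a flap capsule under tight swing and twist limits.
# Bases connect to spine_03 with free angular constraints, as do the upper arms.
# Capsules on pelvis, spine, thighs and upper arms form a coarse body, also joined by free angular constraints.
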PhysicsAsset /BRCosmetics/Characters/Player/Female/Medium/Bodies/F_MED_SpatialTravel_Quirk/Meshes/F_MED_SpatialTravel_Quirk_Physics
bodies {
 dyn_jacket_bk_01 [
  {
   shapeType=Capsule
   length=7
   radius=1.5
   shapeName=dyn_jacket_bk_01_capsule
   center=<5.3958 0.0039 -0.876>
   rotation=<rot 0 -89.815 -96.2553>
  }
  {shapeType=Capsule, length=9, radius=1.5, shapeName=dyn_jacket_bk_01_capsule, center=<8.5 0.1281 1.0821>, rotation=<rot 0.0007 0 93.7715>}
 ]
 dyn_jacket_bk_basa [{shapeType=Capsule, length=0.505, radius=0.505, shapeName=dyn_jacket_bk_basa_capsule, rotation=<rot 0 -90 0>}]
 dyn_jacket_bk_l_01 [{
   shapeType=Capsule
   length=6
   radius=1
   shapeName=dyn_jacket_bk_l_01_capsule
   center=<4.073 -0.253 -0.8089>
   rotation=<rot -7.391 90 -87.9739>
  }]
 dyn_jacket_bk_l_basa [{shapeType=Capsule, length=0.505, radius=0.505, shapeName=dyn_jacket_bk_l_basa_capsule, rotation=<rot 0 -90 0>}]
 dyn_jacket_bk_r_01 [{
   shapeType=Capsule
   length=5.3
   radius=1
   shapeName=dyn_jacket_bk_r_01_capsule
   center=<4.2672 0.6111 -0.5663>
   rotation=<rot 0 -89.2287 -90.7318>
  }]
 dyn_jacket_bk_r_basa [{shapeType=Capsule, length=0.505, radius=0.505, shapeName=dyn_jacket_bk_r_basa_capsule, rotation=<rot 0 -90 0>}]
 dyn_jacket_fr_l_01 [{shapeType=Capsule, length=10, radius=0.505, shapeName=dyn_jacket_fr_l_01_capsule, center=<5.613 -0.8904 0>, rotation=<rot 0 80.7815 89.9999>}]
 dyn_jacket_fr_l_basa [{shapeType=Capsule, length=0.505, radius=0.505, shapeName=dyn_jacket_fr_l_basa_capsule, rotation=<rot 0 -90 0>}]
 dyn_jacket_fr_r_01 [{shapeType=Capsule, length=10, radius=0.505, shapeName=dyn_jacket_fr_r_01_capsule, center=<5.3288 0.8611 0>, rotation=<rot 0 -80.7815 90>}]
 dyn_jacket_fr_r_basa [{shapeType=Capsule, length=0.505, radius=0.505, shapeName=dyn_jacket_fr_r_basa_capsule, rotation=<rot 0 -90 0>}]
 dyn_jacket_l_01 [{
   shapeType=Capsule
   length=8
   radius=0.8
   shapeName=dyn_jacket_l_01_capsule
   center=<7.491 0.3999 0.6844>
   rotation=<rot 3.481 4.9931 -83.1239>
  }]
 dyn_jacket_l_basa [{shapeType=Capsule, length=0.505, radius=0.505, shapeName=dyn_jacket_l_basa_capsule, rotation=<rot 0 -90 0>}]
 dyn_jacket_r_01 [{
   shapeType=Capsule
   length=3
   radius=0.8
   shapeName=dyn_jacket_r_01_capsule
   center=<7.0865 -0.4043 0.1734>
   rotation=<rot -0.3763 -4.1455 -98.6944>
  }]
 dyn_jacket_r_basa [{shapeType=Capsule, length=0.505, radius=0.505, shapeName=dyn_jacket_r_basa_capsule, rotation=<rot 0 -90 0>}]
 pelvis [{shapeType=Capsule, length=10, radius=12, shapeName=pelvis_capsule, center=<-2.8813 -0.0046 0>}]
 spine_01 [
  {shapeType=Capsule, length=4, radius=7.6, shapeName=spine_01_capsule, center=<4.3415 0.9569 0>, rotation=<rot 0 -90 0>}
  {shapeType=Capsule, length=5, radius=7.6, shapeName=spine_01_capsule, center=<-0.5597 -0.3424 0>}
 ]
 spine_02 [{shapeType=Capsule, length=5.5, radius=6, shapeName=spine_02_capsule, center=<2.4975 -0.2168 0>, rotation=<rot 0 -90 0>}]
 spine_03 [{shapeType=Capsule, length=4, radius=8.5, shapeName=spine_03_capsule, center=<0 1.6865 0.2118>}]
 thigh_l [{shapeType=Capsule, length=30, radius=8, shapeName=thigh_l_capsule, center=<-19.6745 0 0>, rotation=<rot 90 0 0>}]
 thigh_r [{shapeType=Capsule, length=30, radius=8, shapeName=thigh_r_capsule, center=<19.6745 0 0>, rotation=<rot 90 0 0>}]
 upperarm_l [{shapeType=Capsule, length=16, radius=6, shapeName=upperarm_l_capsule, center=<13.7184 -0.2152 -0.4898>, rotation=<rot 6.3883 -90 90>}]
 upperarm_r [{shapeType=Capsule, length=16, radius=5, shapeName=upperarm_r_capsule, center=<-13.7184 0.3003 0.3597>, rotation=<rot 7.9938 -90 90>}]
}
constraints [
 {bone1Name=dyn_jacket_bk_01, bone2Name=dyn_jacket_bk_basa, swing1LimitDegrees=5, swing2LimitDegrees=25, twistLimitDegrees=5}
 {bone1Name=dyn_jacket_bk_l_01, bone2Name=dyn_jacket_bk_l_basa, swing1LimitDegrees=5, swing2LimitDegrees=25, twistLimitDegrees=5}
 {bone1Name=dyn_jacket_bk_r_01, bone2Name=dyn_jacket_bk_r_basa, swing1LimitDegrees=5, swing2LimitDegrees=25, twistLimitDegrees=5}
 {bone1Name=dyn_jacket_fr_l_01, bone2Name=dyn_jacket_fr_l_basa, swing1LimitDegrees=5, swing2LimitDegrees=20, twistLimitDegrees=5}
 {bone1Name=dyn_jacket_fr_r_01, bone2Name=dyn_jacket_fr_r_basa, swing1LimitDegrees=5, swing2LimitDegrees=20, twistLimitDegrees=5}
 {bone1Name=dyn_jacket_l_01, bone2Name=dyn_jacket_l_basa, swing1LimitDegrees=10, swing2LimitDegrees=20, twistLimitDegrees=5}
 {bone1Name=dyn_jacket_r_01, bone2Name=dyn_jacket_r_basa, swing1LimitDegrees=0, swing2LimitDegrees=10, twistLimitDegrees=0}
 {
  bone1Name=dyn_jacket_fr_r_basa
  bone2Name=spine_03
  swing1LimitDegrees=45
  swing1Motion=Free
  swing2LimitDegrees=45
  swing2Motion=Free
  twistLimitDegrees=45
  twistMotion=Free
 }
 {
  bone1Name=dyn_jacket_fr_l_basa
  bone2Name=spine_03
  swing1LimitDegrees=45
  swing1Motion=Free
  swing2LimitDegrees=45
  swing2Motion=Free
  twistLimitDegrees=45
  twistMotion=Free
 }
 {
  bone1Name=dyn_jacket_bk_basa
  bone2Name=spine_03
  swing1LimitDegrees=45
  swing1Motion=Free
  swing2LimitDegrees=45
  swing2Motion=Free
  twistLimitDegrees=45
  twistMotion=Free
 }
 {
  bone1Name=dyn_jacket_bk_r_basa
  bone2Name=spine_03
  swing1LimitDegrees=45
  swing1Motion=Free
  swing2LimitDegrees=45
  swing2Motion=Free
  twistLimitDegrees=45
  twistMotion=Free
 }
 {
  bone1Name=dyn_jacket_bk_l_basa
  bone2Name=spine_03
  swing1LimitDegrees=45
  swing1Motion=Free
  swing2LimitDegrees=45
  swing2Motion=Free
  twistLimitDegrees=45
  twistMotion=Free
 }
 {
  bone1Name=dyn_jacket_r_basa
  bone2Name=spine_03
  swing1LimitDegrees=45
  swing1Motion=Free
  swing2LimitDegrees=45
  swing2Motion=Free
  twistLimitDegrees=45
  twistMotion=Free
 }
 {
  bone1Name=dyn_jacket_l_basa
  bone2Name=spine_03
  swing1LimitDegrees=45
  swing1Motion=Free
  swing2LimitDegrees=45
  swing2Motion=Free
  twistLimitDegrees=45
  twistMotion=Free
 }
 {
  bone1Name=upperarm_r
  bone2Name=spine_03
  swing1LimitDegrees=45
  swing1Motion=Free
  swing2LimitDegrees=45
  swing2Motion=Free
  twistLimitDegrees=45
  twistMotion=Free
 }
 {
  bone1Name=upperarm_l
  bone2Name=spine_03
  swing1LimitDegrees=45
  swing1Motion=Free
  swing2LimitDegrees=45
  swing2Motion=Free
  twistLimitDegrees=45
  twistMotion=Free
 }
 {
  bone1Name=spine_03
  bone2Name=spine_02
  swing1LimitDegrees=45
  swing1Motion=Free
  swing2LimitDegrees=45
  swing2Motion=Free
  twistLimitDegrees=45
  twistMotion=Free
 }
 {
  bone1Name=spine_02
  bone2Name=spine_01
  swing1LimitDegrees=45
  swing1Motion=Free
  swing2LimitDegrees=45
  swing2Motion=Free
  twistLimitDegrees=45
  twistMotion=Free
 }
 {
  bone1Name=spine_01
  bone2Name=pelvis
  swing1LimitDegrees=45
  swing1Motion=Free
  swing2LimitDegrees=45
  swing2Motion=Free
  twistLimitDegrees=45
  twistMotion=Free
 }
 {
  bone1Name=thigh_l
  bone2Name=pelvis
  swing1LimitDegrees=45
  swing1Motion=Free
  swing2LimitDegrees=45
  swing2Motion=Free
  twistLimitDegrees=45
  twistMotion=Free
 }
 {
  bone1Name=thigh_r
  bone2Name=pelvis
  swing1LimitDegrees=45
  swing1Motion=Free
  swing2LimitDegrees=45
  swing2Motion=Free
  twistLimitDegrees=45
  twistMotion=Free
 }
]
